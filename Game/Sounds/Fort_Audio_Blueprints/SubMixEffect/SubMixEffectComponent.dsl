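; SubMixEffectComponent is an audio Blueprint that toggles a single submix effect across a set of submix assets.
; Enable adds the SubMixEffect to every entry in SubMixAssets, and Disable removes it from each one.
; Both functions check and update a bEnabled flag so repeated calls do nothing, and getters expose the effect and enabled state.
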(fn Disable
  (if (Variables|Default|GetbEnabled)
    (Variables|Default|SetbEnabled false)
    (for _array_element (Variables|Default|GetSubMixAssets)
      (Audio|Effects|RemoveSubmixEffect _array_element (Variables|Default|GetSubMixEffect)))))

(fn Enable
  (if (not (Variables|Default|GetbEnabled))
    (Variables|Default|SetbEnabled true)
    (for _array_element (Variables|Default|GetSubMixAssets)
      (Audio|Effects|AddSubmixEffect _array_element (Variables|Default|GetSubMixEffect)))))

(fn GetEffect
  (return (Variables|Default|GetSubMixEffect)))

(fn GetEnabled
  (return (Variables|Default|GetbEnabled)))
